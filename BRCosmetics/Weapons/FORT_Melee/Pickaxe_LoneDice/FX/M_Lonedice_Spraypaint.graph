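# Translucent unlit Niagara sprite material for the LoneDice pickaxe's spray-paint splat effect.
# A radial sphere mask, its radius distorted by cell noise, is smoothstepped and eroded by white noise to form the splat, driven by dynamic parameters.
# Particle color darkens toward the edge, and the splat times particle alpha sets opacity.
# A stabilized screen-space grunge sample from the LoneDice outline spray atlas is subtracted from the emissive color.
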Material /BRCosmetics/Weapons/FORT_Melee/Pickaxe_LoneDice/FX/M_Lonedice_Spraypaint
graph {
 properties {
  bAutomaticallySetUsageInEditor=false
  bUsedWithNiagaraSprites=true
  blendMode=BLEND_Translucent
  shadingModel=MSM_Unlit
  usedShadingModels=Unlit
 }
 nodes [
  n1 = ParticleColor()
  n2 = DynamicParameter() {defaultValue=<rgba 3 1 1 0.5>, paramNames [SpherePower, SphereStrenght, NoiseStrenght, ColorDarkeningAmount]}
  n3 = DynamicParameter() {parameterIndex=1, defaultValue=<rgba 0 0.2 0.5 0.5>, paramNames [SplatSmoothMin, SplatSmoothMax, DistortNoiseXTile, DistortNoiseYTile]}
  n4 = Multiply(A=TextureCoordinate(), B=AppendVector(A=@n3.DistortNoiseXTile, B=@n3.DistortNoiseXTile))
  n5 = DynamicParameter() {parameterIndex=2, defaultValue=<rgba 1 1 0.25 0.8>, paramNames [NoiseOffsetX, NoiseOffsetY, NoiseStrenght, SphereRadius]}
  n6 = AppendVector(A=@n5.NoiseOffsetX, B=@n5.NoiseOffsetX)
  n7 = Multiply(A=TextureSample(UVs=Add(A=@n4, B=@n6)) {texture=/Game/Creative/PostProcess/Textures/Cell_Noise_T1}.R, B=@n5.NoiseStrenght) {constB=0.25}
  n8 = Power(
   Base=SphereMask(A=TextureCoordinate(), B=Constant2Vector() {g=0.5, r=0.5}, Radius=Add(A=@n5.SphereRadius, B=@n7), Hardness=Constant())
   Exp=@n2.SpherePower
  )
  n9 = DynamicParameter() {parameterIndex=1, defaultValue=<rgba 0 0.2 0.5 0.5>, paramNames [SplatSmoothMin, SplatSmoothMax, DistortNoiseXTile, DistortNoiseYTile]}
  n10 = Multiply(A=@n8, B=@n2.SphereStrenght) {constB=2}
  n11 = TextureSample(UVs=TextureCoordinate()) {samplerType=SAMPLERTYPE_LinearGrayscale, texture=/Game/Global/Textures/Default/Noise/T_WhiteNoise}
  n12 = DynamicParameter() {defaultValue=<rgba 3 1 1 0.5>, paramNames [SpherePower, SphereStrenght, NoiseStrenght, ColorDarkeningAmount]}
  n13 = Saturate(None=Subtract(
    A=SmoothStep(Min=@n9.SplatSmoothMin, Max=@n9.SplatSmoothMin, Value=@n10) {constMax=0.2, constMin=0.2}
    B=Multiply(A=@n11.R, B=@n12.NoiseStrenght) {constB=0.5}
   ))
  n14 = Multiply(A=LinearInterpolate(A=Multiply(A=@n1.RGB, B=@n2.ColorDarkeningAmount) {constB=0.5}, B=@n1.RGB, Alpha=Saturate(None=@n8)), B=@n13)
  n15 = TextureSample(UVs=MaterialFunctionCall!(Position, UVRadialScale=ScalarParameter() {defaultValue=0.3, group=LoneDice Graffiti Outline, parameterName=LD_GrungeMap_Scale}) {materialFunction=/Game/Characters/Player/Common/MaterialFunctions/Superlevel/CharacterStabilizedScreenSpaceUVs}.StabilizedUVs) {
   samplerType=SAMPLERTYPE_Masks
   texture=/BRCosmetics/Characters/Player/Male/Medium/Bodies/M_MED_LoneDice/Textures/T_LoneDice_Outline_SprayAtlas
  }
  n16 = Multiply(A=@n15.G, B=ScalarParameter() {group=LoneDice Graffiti Outline, parameterName=LD_GrungeMap_Intensity})
  MaterialFunctionCall(Emissive=Saturate(None=Subtract(A=@n14, B=@n16))) {materialFunction=&MF_FX_EmissiveAdjustments}
  Saturate(None=Multiply(A=@n1.A, B=@n13))
 ]
}
