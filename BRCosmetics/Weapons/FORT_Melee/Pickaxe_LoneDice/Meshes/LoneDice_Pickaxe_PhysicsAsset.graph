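# Physics asset for the LoneDice pickaxe mesh, a two-handed melee weapon cosmetic.
# It defines a single capsule body on the Melee_TwoHanded_Handle bone, offset and rotated from the bone origin.
# It has no constraints, so no joints link it to other bodies.
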PhysicsAsset /BRCosmetics/Weapons/FORT_Melee/Pickaxe_LoneDice/Meshes/LoneDice_Pickaxe_PhysicsAsset
bodies {Melee_TwoHanded_Handle [{shapeType=Capsule, length=40.4041, radius=16.6716, center=<-37.0101 -10.9227 30.7087>, rotation=<rot 0 -90.0016 -85.6496>}]}
constraints []
